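# Material function that builds the base asphalt layer for the Neon City set, outputting material attributes plus two UV sets.
# Diffuse, mask and normal textures feed the attributes, with optional metallic from the mask's green channel and scalable roughness.
# Asphalt UVs switch between tiled world XY (default) and mesh coordinates, while patch UVs use separately tiled, offset world XY.
# The normal's red channel is also exposed as a PebbleMask reroute.
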MaterialFunction /Game/Environments/Asteria/Sets/NeonCity/MaterialFunctions/MF_NeonCity_Asphalt
graph {nodes [
  n1 = TextureSampleParameter2D(UVs=NamedRerouteUsage()) {
   group=Asphalt
   parameterName=Diffuse
   samplerSource=SSM_Wrap_WorldGroupSettings
   sortPriority=2
   texture=/Game/Athena/Apollo/Environments/Landscape/Textures/Roads/T_AsphaltDetail1_D
  }
  n2 = TextureSampleParameter2D(UVs=NamedRerouteUsage()) {
   group=Asphalt
   parameterName=Masks
   samplerSource=SSM_Wrap_WorldGroupSettings
   samplerType=SAMPLERTYPE_Masks
   sortPriority=3
   texture=/Game/Athena/Apollo/Environments/Landscape/Textures/Roads/T_AsphaltDetail1_S
  }
  n3 = TextureSampleParameter2D(UVs=NamedRerouteUsage()) {
   group=Asphalt
   parameterName=Normals
   samplerSource=SSM_Wrap_WorldGroupSettings
   samplerType=SAMPLERTYPE_Normal
   sortPriority=4
   texture=/Game/Athena/Apollo/Environments/Landscape/Textures/Roads/T_AsphaltDetail1_N
  }
  n4 = SetMaterialAttributes!(
   MaterialAttributes
   BaseColor=@n1.RGB
   Metallic=StaticSwitchParameter(True=@n2.G, False=Constant()) {parameterName=UseMetallic}
   Specular=@n2.R
   Roughness=Multiply(A=@n2.B, B=ScalarParameter() {defaultValue=1, group=Puddles, parameterName=RoughnessScale})
   Normal=@n3.RGB
  ) {attributeSetTypes [
    #1
    #2
    #3
    #4
    #5
   ]}
  FunctionOutput(None=@n4)
  n5 = NamedRerouteDeclaration(None=ComponentMask(None=WorldPosition() {worldPositionShaderOffset=WPT_ExcludeAllShaderOffsets}.XYZ) {g=true, r=true}) {name=WorldPos.XY, nodeColor=<rgba 1 1 0 1>}
  n6 = Divide(A=@n5, B=ScalarParameter() {defaultValue=120, group=Asphalt, parameterName=WorldTileDivisor_Asphalt, sortPriority=1})
  n7 = TextureCoordinate()
  FunctionOutput(None=NamedRerouteDeclaration(None=StaticSwitchParameter(True=@n6, False=@n7) {defaultValue=true, parameterName=IsWorldAligned}) {name=Asphalt UVs, nodeColor=<rgba 0 0.9531 1 1>}) {outputName=Asphalt UVs, sortPriority=10}
  n8 = Divide(A=@n5, B=ScalarParameter() {defaultValue=50, group=Asphalt, parameterName=WorldTileDivisor_Asphalt_Patches, sortPriority=1})
  n9 = AppendVector(
   A=ScalarParameter() {defaultValue=0.3, group=AsphaltPatches, parameterName=AsphaltPatchOffset_U}
   B=ScalarParameter() {defaultValue=30, group=AsphaltPatches, parameterName=AsphaltPatchOffset_V}
  )
  FunctionOutput(None=NamedRerouteDeclaration(None=Add(A=@n8, B=@n9) {constB=0.5}) {name=AsphaltPatchOffsetUVs, nodeColor=<rgba 0.875 1 0 1>}) {outputName=Asphalt Patches UVs, sortPriority=20}
  NamedRerouteDeclaration(None=ComponentMask(None=@n3.RGB) {r=true}) {name=PebbleMask, nodeColor=<rgba 0.4802 0.4769 0.4696 1>}
 ]}
