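# Material function that adds wet areas and water puddles to NeonCity building material attributes.
# A world-XY projected mask texture, optionally limited to material layers, is contrast-remapped and multiplied by upward-facing vertex normals into Wet and Puddle masks.
# These masks drive base color darkening, roughness, platform-dependent specular and flattening of normals to world up, with optional metallic-based roughness retention.
# Static switches bypass the effect entirely or on Low quality and mobile shading.
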MaterialFunction /Game/Environments/Asteria/Sets/NeonCity/MaterialFunctions/MF_WaterPuddlesBuilding
graph {nodes [
  n1 = FunctionInput() {inputType=FunctionInput_MaterialAttributes}
  n2 = GetMaterialAttributes(None=@n1) {attributeGetTypes [
    #1
    #2
    #3
    #4
    #5
   ]}
  n3 = NamedRerouteUsage()
  n4 = LinearInterpolate(A=@n2.Roughness, B=ScalarParameter() {group=WaterPuddles, parameterName=Puddle_Roughness}, Alpha=@n3)
  n5 = Saturate(None=Multiply(A=@n2.Metallic, B=ScalarParameter() {defaultValue=1, parameterName=Puddle_Roughness_OrigianlMetal}))
  n6 = ScalarParameter() {defaultValue=1, group=WaterPuddles, parameterName=Puddle_Specular}
  n7 = ScalarParameter() {defaultValue=0.1, parameterName=Puddle_Specular_Low}
  n8 = MaterialFunctionCall("Xb1,PS4"=@n6, "PC-High&&Epic,XSX,PS5"=@n6, "PC-Perf,Android"=@n7, Switch=@n6, "PC-Low(NotPerf)&&Medium"=@n7) {materialFunction=&PlatformSwitch_1D}
  n9 = NamedRerouteUsage()
  n10 = Constant3Vector() {constant=<rgba 0 0 1 0>}
  n11 = ScalarParameter() {defaultValue=0.6, group=WaterPuddles, parameterName=PuddleWet_Darken}
  n12 = NamedRerouteUsage()
  n13 = SetMaterialAttributes!(
   MaterialAttributes=@n2.MaterialAttributes
   Roughness=StaticSwitchParameter(True=LinearInterpolate(A=@n4, B=@n2.Roughness, Alpha=@n5), False=@n4) {parameterName=WaterPuddle_UseMetallic}
   Specular=Min(A=LinearInterpolate(A=@n2.Specular, B=@n8.Result, Alpha=@n9)) {constB=0.5}
   Normal=LinearInterpolate(A=@n2.Normal, B=Transform(None=@n10) {transformSourceType=TRANSFORMSOURCE_World, transformType=TRANSFORM_Tangent}, Alpha=NamedRerouteUsage())
   BaseColor=Multiply(A=@n2.BaseColor, B=LinearInterpolate(B=@n11, Alpha=@n12) {constA=1})
  ) {attributeSetTypes [
    #1
    #2
    #3
    #4
   ]}
  n14 = QualitySwitch(Default=ShadingPathSwitch(Default=@n13, Mobile=@n1) {desc=Migrated from FeatureLevelSwitch}, Low=@n1)
  FunctionOutput(None=StaticSwitchParameter(True=StaticSwitchParameter(True=@n14, False=@n13) {parameterName=DisableWaterPuddlesOnLOW}, False=@n1) {defaultValue=true, parameterName=UseWaterPuddlesBuildings})
  Min(
   A=@n4
   B=LinearInterpolate(A=@n2.Roughness, B=ScalarParameter() {defaultValue=1, group=WaterPuddles, parameterName=Puddle_Roughness_min}, Alpha=@n3)
  )
  n15 = WorldPosition() {worldPositionShaderOffset=WPT_ExcludeAllShaderOffsets}
  n16 = TextureSampleParameter2D(UVs=Divide(A=ComponentMask(None=@n15.XYZ) {g=true, r=true}, B=ScalarParameter() {defaultValue=2000, group=WaterPuddles, parameterName=Puddle_Size})) {
   group=WaterPuddles
   parameterName=Puddle Mask
   samplerType=SAMPLERTYPE_LinearColor
   texture=/Game/Environments/Asteria/Sets/NeonCity/Textures/WaterPuddles_M
  }
  n17 = MaterialFunctionCall!(
   1=ScalarParameter() {defaultValue=1, group=WaterPuddles, parameterName=Water Puddle Layer 1}
   2=ScalarParameter() {group=WaterPuddles, parameterName=Water Puddle Layer 2}
   3=ScalarParameter() {group=WaterPuddles, parameterName=Water Puddle Layer 3}
   4=ScalarParameter() {group=WaterPuddles, parameterName=Water Puddle Layer 4}
  ) {materialFunction=/Game/Global/Base4Layers_2022/MaterialFunctions/MF_UV2LerpScalar_Layers_New}
  n18 = StaticSwitchParameter(True=LinearInterpolate(B=@n16.R, Alpha=Saturate(None=@n17.Result)), False=@n16.R) {group=WaterPuddles, parameterName=Limit Water Puddles to Specific Layers}
  n19 = Saturate(None=LinearInterpolate(
    A=ScalarParameter() {defaultValue=0.4208, group=WaterPuddles, parameterName=PuddleWetness_Contrast_Low}
    B=ScalarParameter() {defaultValue=2.1504, group=WaterPuddles, parameterName=PuddleWetness_Contrast_High}
    Alpha=@n18
   ))
  n20 = ComponentMask(None=VertexNormalWS()) {b=true}
  NamedRerouteDeclaration(None=Multiply(A=Saturate(None=@n19), B=Saturate(None=@n20))) {name=WetMask, nodeColor=<rgba 0 1 0.4844 1>}
  n21 = LinearInterpolate(
   A=ScalarParameter() {defaultValue=-6.4879, group=WaterPuddles, parameterName=Puddle_Contrast_Low}
   B=ScalarParameter() {defaultValue=17.5196, group=WaterPuddles, parameterName=Puddle_Contrast_High}
   Alpha=@n18
  )
  n22 = LinearInterpolate(
   A=ScalarParameter() {defaultValue=-78.5354, group=WaterPuddles, parameterName=PuddleSlopeMask_Low}
   B=ScalarParameter() {defaultValue=1.2891, group=WaterPuddles, parameterName=PuddleSlopeMask_High}
   Alpha=@n20
  )
  NamedRerouteDeclaration(None=Multiply(A=Saturate(None=@n21), B=Saturate(None=@n22))) {name=PuddleMask, nodeColor=<rgba 1 0.8438 0 1>}
 ]}
