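# Material function that generates an animated rain-ripple normal, named for water surfaces in the NeonCity set.
# It samples the packed RainDrops_M mask at the input UV, using blue for per-drop time offsets and alpha as a ripple gradient.
# Time drives a sine wave and a looping frac-based fade envelope, shaped by a remapped mask.
# The saturated result blends from a flat normal toward a direction derived from texture coordinates minus the red channel.
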MaterialFunction /Game/Environments/Asteria/Sets/NeonCity/MaterialFunctions/MF_WaterRainRipple
graph {nodes [
  n1 = Constant() {r=1}
  n2 = NamedRerouteUsage()
  n3 = NamedRerouteUsage()
  n4 = NamedRerouteUsage()
  n5 = NamedRerouteUsage()
  n6 = ScalarParameter() {defaultValue=1.5386, parameterName=RippleFrequency}
  n7 = NamedRerouteUsage()
  n8 = ScalarParameter() {defaultValue=1.487, parameterName=RippleSpeed}
  n9 = Sine(None=Add(A=Multiply(A=@n5, B=@n6), B=Multiply(A=@n7, B=@n8)))
  n10 = Multiply(A=Multiply(A=LinearInterpolate(A=@n1, B=@n2, Alpha=@n3), B=Multiply(A=@n4, B=@n9)), B=@n3)
  FunctionOutput(None=LinearInterpolate(A=Constant3Vector() {constant=<rgba 0 0 1 0>}, B=NamedRerouteUsage(), Alpha=Saturate(None=@n10))) {outputName=Normal}
  n11 = TextureSample(UVs=FunctionInput() {inputName=UV, inputType=FunctionInput_Vector2}) {samplerType=SAMPLERTYPE_Masks, texture=/Game/Environments/Asteria/Sets/NeonCity/Techart/Textures/RainDrops_M}
  NamedRerouteDeclaration(None=@n11.B) {name=RippleTimeOffsetTexture, nodeColor=<rgba 0 1 0.6484 1>}
  n12 = NamedRerouteUsage()
  n13 = ScalarParameter() {defaultValue=3.6936, parameterName=RippleTimeOffset}
  NamedRerouteDeclaration(None=Add(A=Multiply(A=@n12, B=@n13), B=Time())) {name=RippleTime, nodeColor=<rgba 0 0.8359 1 1>}
  NamedRerouteDeclaration(None=@n11.A) {name=RippleLinearGradient, nodeColor=<rgba 1 0 0.375 1>}
  n14 = ScalarParameter() {defaultValue=-1.9809, parameterName=RippleMask_low}
  n15 = ScalarParameter() {defaultValue=9.3212, parameterName=RippleMask_High}
  n16 = NamedRerouteUsage()
  NamedRerouteDeclaration(None=Saturate(None=LinearInterpolate(A=@n14, B=@n15, Alpha=@n16))) {name=FinalMask, nodeColor=<rgba 0.7578 0 1 1>}
  n17 = Frac(None=Multiply(A=NamedRerouteUsage(), B=ScalarParameter() {defaultValue=0.1351, parameterName=RippleAnimSpeed}))
  n18 = ScalarParameter() {defaultValue=-0.3808, parameterName=RippleFadeIn}
  n19 = Constant() {r=1}
  n20 = Saturate(None=Divide(A=Subtract(A=@n17, B=@n18), B=Subtract(A=@n19, B=@n18)))
  NamedRerouteDeclaration(None=Multiply(A=@n20, B=OneMinus(None=@n20))) {name=RippleTimeOffsetANim, nodeColor=<rgba 1 0 0.7969 1>}
  n21 = AppendVector(A=Subtract(A=TextureCoordinate(), B=AppendVector(A=@n11.R, B=@n11.R)), B=Constant())
  NamedRerouteDeclaration(None=Multiply(A=ScalarParameter() {defaultValue=0.512, parameterName=RippleIntensity}, B=Normalize(VectorInput=@n21))) {name=RippleNormal, nodeColor=<rgba 1 0 0.3281 1>}
 ]}
