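# Material function that adds rain wetness and animated water drops to NeonCity material attributes, with a switch named for walls.
# World-projected noise and panning drop normals from a heightmap, optionally scaled by time of day, build wet masks.
# These masks darken base color and adjust roughness, specular and normal, with cheaper low-quality and mobile paths.
# Optional vertex color, UV1 layer and object-space gradient masks limit wetness; the gradient is also output.
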MaterialFunction /Game/Environments/Asteria/Sets/NeonCity/MaterialFunctions/MF_WaterWetness
graph {nodes [
  n1 = FunctionInput() {inputType=FunctionInput_MaterialAttributes}
  n2 = GetMaterialAttributes(None=@n1) {attributeGetTypes [
    #1
    #2
    #3
    #4
    #5
   ]}
  n3 = Min(
   A=LinearInterpolate(A=@n2.Roughness, B=ScalarParameter() {group=WaterPuddles, parameterName=Wet_Roughness}, Alpha=NamedRerouteUsage())
   B=@n2.Roughness
  )
  n4 = Saturate(None=Multiply(A=@n2.Metallic, B=ScalarParameter() {defaultValue=1, parameterName=Puddle_Roughness_OrigianlMetal}))
  n5 = ScalarParameter() {defaultValue=1, group=WaterPuddles, parameterName=Wet_Specular}
  n6 = NamedRerouteUsage()
  n7 = MaterialFunctionCall(BaseNormal=@n2.Normal, AdditionalNormal=NamedRerouteUsage()) {materialFunction=&BlendAngleCorrectedNormals}
  n8 = ScalarParameter() {defaultValue=0.6, group=WaterPuddles, parameterName=PuddleWet_Darken}
  n9 = NamedRerouteUsage()
  n10 = SetMaterialAttributes!(
   MaterialAttributes=@n2.MaterialAttributes
   Roughness=StaticSwitchParameter(True=LinearInterpolate(A=@n3, B=@n2.Roughness, Alpha=@n4), False=@n3) {parameterName=UseMetallic}
   Specular=Min(A=LinearInterpolate(A=@n2.Specular, B=@n5, Alpha=@n6)) {constB=0.5}
   Normal=QualitySwitch(Default=ShadingPathSwitch(Default=@n7.Result, Mobile=@n2.Normal) {desc=Migrated from FeatureLevelSwitch}, Low=@n2.Normal)
   BaseColor=Multiply(A=@n2.BaseColor, B=LinearInterpolate(B=@n8, Alpha=@n9) {constA=1})
  ) {attributeSetTypes [
    #1
    #2
    #3
    #4
   ]}
  n11 = StaticSwitchParameter(
   True=BlendMaterialAttributes(A=@n1, B=@n10, Alpha=NamedRerouteUsage())
   False=BlendMaterialAttributes(A=@n1, B=@n10, Alpha=NamedRerouteUsage())
  ) {group=WetMasking, parameterName=UseWetDryTransition}
  n12 = VertexColor()
  n13 = BlendMaterialAttributes(
   A=@n1
   B=@n11
   Alpha=ChannelMaskParameter(None=AppendVector(A=@n12, B=@n12)) {group=WetMasking, parameterName=WaterWetness_VertexColorChannelMask}.R
  ) {vertexAttributeBlendType=UseA}
  FunctionOutput(None=StaticSwitchParameter(True=StaticSwitchParameter(True=@n13, False=@n11) {group=WetMasking, parameterName=UseWaterWetness_VertexColorMask}, False=@n1) {defaultValue=true, group=WaterPuddles, parameterName=UseWaterWetness_Walls})
  n14 = DotProduct(
   A=Transform(None=VectorParameter() {group=WetMasking, parameterName=WetDry_Gradient Direction, defaultValue=<rgba 0 1 0 1>}.RGB) {transformSourceType=TRANSFORMSOURCE_Local}
   B=Subtract(A=WorldPosition().XYZ, B=ObjectPositionWS())
  )
  n15 = Multiply(A=ScalarParameter() {defaultValue=1, group=WetMasking, parameterName=WetDry_GradientLength}, B=ObjectRadius())
  n16 = Divide(A=@n15)
  n17 = Power(Base=Divide(A=Add(A=@n14, B=@n16), B=@n15), Exp=ScalarParameter() {defaultValue=1, group=WetMasking, parameterName=WetDry_Gradient_Falloff})
  FunctionOutput(None=NamedRerouteDeclaration(None=Saturate(None=@n17)) {name=WetDryMask, nodeColor=<rgba 1 0 0.1172 1>}) {outputName=DryWetMaskTransition, sortPriority=10}
  n18 = ScalarParameter() {parameterName=WetRain_TODM_Morning_Evening}
  n19 = ScalarParameter() {parameterName=WetRain_TODM_Day}
  n20 = ScalarParameter() {defaultValue=1, parameterName=WetRain_TODM_Night}
  n21 = ComponentMask(None=VertexNormalWS()) {b=true}
  n22 = ScalarParameter() {defaultValue=-1, parameterName=WetMask_ProjectionTop_Contrast2}
  n23 = TextureObject() {samplerType=SAMPLERTYPE_Masks, texture=/Game/Environments/Asteria/Sets/NeonCity/Textures/WaterWetness_M}
  n24 = ScalarParameter() {defaultValue=0.15, parameterName=WaterDrop_Normal}
  n25 = ScalarParameter() {defaultValue=0.008, parameterName=WaterDrop_Offset}
  n26 = Constant2Vector()
  n27 = Multiply(A=Time(), B=ScalarParameter() {defaultValue=-0.63, parameterName=WaterDropPanSpeed})
  n28 = WorldPosition() {worldPositionShaderOffset=WPT_ExcludeAllShaderOffsets}
  n29 = ScalarParameter() {defaultValue=-120, parameterName=DropMask_Size}
  n30 = Multiply(A=NamedRerouteUsage(), B=ScalarParameter() {defaultValue=-0.3382, parameterName=WaterDropDistort})
  n31 = Add(A=Add(A=AppendVector(A=@n26, B=@n27), B=Divide(A=@n28.XYZ, B=@n29)), B=AppendVector(A=AppendVector(A=@n30, B=@n30), B=Constant()))
  n32 = Constant4Vector() {constant=<rgba 0 0 0 1>}
  n33 = LinearInterpolate(
   A=MaterialFunctionCall!(
    Height Map=@n23
    Normal Map Intensity=@n24
    Height Map UV Offset=@n25
    Coordinates=ComponentMask(None=@n31) {b=true, r=true}
    Height Map Channel Selector=@n32.RGBA
   ) {materialFunction=/Engine/Functions/Engine_MaterialFunctions03/Procedurals/NormalFromHeightmap}.Result
   B=MaterialFunctionCall(
    Height Map=@n23
    Normal Map Intensity=@n24
    Height Map UV Offset=@n25
    Coordinates=ComponentMask(None=@n31) {b=true, g=true}
    Height Map Channel Selector=@n32.RGBA
   ) {materialFunction=&NormalFromHeightmap}.Result
   Alpha=NamedRerouteUsage()
  )
  n34 = Multiply(
   A=Multiply(
    A=MaterialFunctionCall(In=@n21, Contrast=@n22) {materialFunction=&CheapContrast}.Result
    B=ComponentMask(None=@n33) {g=true, r=true}
   )
   B=NamedRerouteUsage()
  )
  n35 = StaticSwitchParameter(
   True=Multiply(
    A=MaterialFunctionCall(Value_Morning=@n18, Value_Day=@n19, Value_Evening=@n18, Value_Night=@n20) {materialFunction=&Scalar_Time_of_Day}.Result
    B=@n34
   )
   False=@n34
  ) {defaultValue=true, parameterName=MaskByTOD}
  n36 = Abs(None=@n35)
  n37 = Add(A=ComponentMask(None=@n36) {r=true}, B=ComponentMask(None=@n36) {g=true})
  n38 = LinearInterpolate(
   A=ScalarParameter() {defaultValue=-0.6244, group=WaterPuddles, parameterName=PuddleWetness_Contrast_Low}
   B=ScalarParameter() {defaultValue=0.9387, group=WaterPuddles, parameterName=PuddleWetness_Contrast_High}
   Alpha=NamedRerouteUsage()
  )
  n39 = ShadingPathSwitch(Default=Add(A=Multiply(A=@n37) {constB=5}, B=@n38), Mobile=@n38) {desc=Migrated from FeatureLevelSwitch}
  NamedRerouteDeclaration(None=Saturate(None=QualitySwitch(Default=@n39, Low=@n38))) {name=WetMask, nodeColor=<rgba 0 1 0.4844 1>}
  n40 = ScalarParameter() {defaultValue=2180.7625, parameterName=WetMask_Size}
  n41 = Divide(A=@n28.XYZ, B=AppendVector(A=AppendVector(A=@n40, B=@n40), B=ScalarParameter() {defaultValue=9320.5801, parameterName=WetMask_Size_Z}))
  n42 = TextureSample(UVs=ComponentMask(None=@n41) {b=true, r=true}) {
   samplerSource=SSM_Wrap_WorldGroupSettings
   samplerType=SAMPLERTYPE_Masks
   texture=&WaterWetness_M
  }
  n43 = TextureSample(UVs=ComponentMask(None=@n41) {b=true, g=true}) {
   samplerSource=SSM_Wrap_WorldGroupSettings
   samplerType=SAMPLERTYPE_Masks
   texture=&WaterWetness_M
  }
  n44 = LinearInterpolate(A=@n42.B, B=@n43.B, Alpha=NamedRerouteUsage())
  n45 = TextureSample(UVs=ComponentMask(None=@n41) {g=true, r=true}) {
   samplerSource=SSM_Wrap_WorldGroupSettings
   samplerType=SAMPLERTYPE_Masks
   texture=&WaterWetness_M
  }
  n46 = StaticSwitchParameter(True=@n45.B, False=ScalarParameter() {defaultValue=0.9017, parameterName=TopWetMask}) {group=WaterPuddles, parameterName=Wet_IsTriPlanar}
  n47 = NamedRerouteUsage()
  NamedRerouteDeclaration(None=Multiply(A=LinearInterpolate(A=@n44, B=@n46, Alpha=@n47), B=NamedRerouteUsage())) {name=WetNoiseMask, nodeColor=<rgba 1 0.9609 0 1>}
  NamedRerouteDeclaration(None=@n44) {name=Distortion NOise, nodeColor=<rgba 0.1484 0 1 1>}
  n48 = VertexNormalWS()
  n49 = Abs(None=ComponentMask(None=@n48) {r=true})
  n50 = ScalarParameter() {defaultValue=2, parameterName=WetMask_Projection_Contrast}
  NamedRerouteDeclaration(None=Saturate(None=MaterialFunctionCall(In=@n49, Contrast=@n50) {materialFunction=&CheapContrast}.Result)) {name=XYMask, nodeColor=<rgba 0 0.6719 1 1>}
  n51 = ComponentMask(None=@n48) {b=true}
  NamedRerouteDeclaration(None=MaterialFunctionCall(In=@n51, Contrast=ScalarParameter() {defaultValue=-0.1707, parameterName=WetMask_ProjectionTop_Contrast}) {materialFunction=&CheapContrast}.Result) {name=ZMaskTop, nodeColor=<rgba 1 0.2578 0 1>}
  n52 = Multiply(A=@n51) {constB=-1}
  NamedRerouteDeclaration(None=Saturate(None=OneMinus(None=@n52))) {name=ZMask_Bottom, nodeColor=<rgba 1 0 0.1406 1>}
  NamedRerouteDeclaration(None=AppendVector(A=@n35, B=Constant() {r=1})) {name=DropNormals, nodeColor=<rgba 1 0 0.5156 1>}
  n53 = ComponentMask(None=TextureCoordinate() {coordinateIndex=1}) {r=true}
  n54 = Floor(None=@n53)
  n55 = Abs(None=Subtract(A=@n54, B=Constant() {r=2}))
  n56 = Abs(None=Subtract(A=@n54, B=Constant() {r=1}))
  n57 = Constant() {r=1}
  n58 = StaticSwitch(
   True=If(A=@n53, B=Constant() {r=1}, "A > B"=@n57, "A < B"=Constant()) {desc=0}
   False=@n57
   Value=StaticBoolParameter() {group=WetMasking, parameterName=Wet_Maskout_Layer_1}
  )
  n59 = StaticSwitch(
   True=If(A=@n53, B=Constant() {r=1}, "A > B"=Saturate(None=@n56), "A < B"=@n58) {desc=0}
   False=@n58
   Value=StaticBoolParameter() {group=WetMasking, parameterName=Wet_Maskout_Layer_2}
  )
  n60 = StaticSwitch(
   True=If(A=@n53, B=Constant() {r=2}, "A > B"=Saturate(None=@n55), "A < B"=@n59)
   False=@n59
   Value=StaticBoolParameter() {group=WetMasking, parameterName=Wet_Maskout_Layer_3}
  )
  n61 = Abs(None=Subtract(A=@n54, B=Constant() {r=3}))
  n62 = StaticSwitch(
   True=If(A=@n53, B=Constant() {r=3}, "A > B"=Saturate(None=@n61), "A < B"=@n60)
   False=@n60
   Value=StaticBoolParameter() {group=WetMasking, parameterName=Wet_Maskout_Layer_4}
  )
  NamedRerouteDeclaration(None=@n62) {name=WetLayerMask, nodeColor=<rgba 0.8047 0 1 1>}
  Reroute(None=@n62)
 ]}
